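# Control Rig for the MuffinLadle_Gas player body cosmetic, driving its custom split front/back arm, finger and leg bones from the standard skeleton.
# Front arms follow via rotation constraints and PBIK to the hands, back arms copy them, and extra thumbs follow pinkies.
# Joint-angle remaps weight shoulder plate, clavicle and wire correctives.
# Without shoesEquipped, FABRIK fits custom legs to offset ball targets; a thumb-retargeting chain is unconnected.
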ControlRigBlueprint /BRCosmetics/Characters/Player/Male/Medium/Bodies/M_MED_MuffinLadle_Gas/Meshes/M_MED_MuffinLadle_Gas_CtrlRig
graphs [
 RigVMFunctionLibrary {links [], nodes []}
 RigVMFunction_Sequence_1_ContainedGraph {nodes [
   n1 = RigVMFunctionEntryNode()
   RigVMFunction_Sequence = &RigVMFunction_Sequence(ExecuteContext=@n1.ExecuteContext)
   n2 = &RigVMFunction_Sequence(ExecuteContext=@RigVMFunction_Sequence.B)
   RigVMFunctionReturnNode(A=@RigVMFunction_Sequence.A, B=@n2.A, C=@n2.B)
  ]}
 RigVMModel {nodes [
   For_Each_1 = RigVMDispatchNode(ExecuteContext=(), Array=RigVMVariableNode() {variable=OriginalThumbs}.Value) {template="DISPATCH_RigVMDispatch_ArrayIterator(in Array,out Element,out Index,out Count,out Ratio)"}
   > RigVMVariableNode(Value=@For_Each_1.Element.Keys) {variable=OriginalThumbs Index}
   > RigVMVariableNode(Value=RigVMDispatchNode(Index=@For_Each_1.Index, Array=RigVMVariableNode() {variable=CustomOuterThumbs}.Value) {template="DISPATCH_RigVMDispatch_ArrayGetAtIndex(in Array,in Index,out Element)"}.Element.Keys) {variable=CustomOuterThumbs Index}
   n1 = RigVMDispatchNode(Index=0, Array=RigVMVariableNode() {variable=OriginalThumbs Index}.Value) {template="DISPATCH_RigVMDispatch_ArrayGetAtIndex(in Array,in Index,out Element)"}
   n2 = &RigUnit_GetTransform(Space=LocalSpace, bInitial=False, Item=@n1.Element)
   n3 = /Script/RigVM.RigVMFunction_MathQuaternionInverse(Value=&RigUnit_GetTransform(Space=LocalSpace, bInitial=true, Item=@n1.Element).Transform.Rotation) {is=struct RigVMUnitNode, template="Inverse::Execute(in Value,out Result)"}
   > &RigUnit_ModifyTransforms(
    ItemToModify=((Item=(Type=Bone,Name="None"),Transform=(Rotation=(X=0,Y=0,Z=0,W=1),Translation=(X=0,Y=0,Z=0),Scale3D=(X=1,Y=1,Z=1))))
    Weight=1
    WeightMinimum=0
    ItemToModify.0.Transform.Rotation=/Script/RigVM.RigVMFunction_MathQuaternionMul(
     B=(X=-0.0168,Y=-0.0266,Z=0.0257,W=0.9992)
     A=&RigVMFunction_MathQuaternionMul(A=@n2.Transform.Rotation, B=@n3.Result) {template="Multiply::Execute(in A,in B,out Result)"}.Result
    ) {is=struct RigVMUnitNode, template="Multiply::Execute(in A,in B,out Result)"}.Result
    ItemToModify.0.Item=RigVMDispatchNode(Index=0, Array=RigVMVariableNode() {variable=CustomOuterThumbs Index}.Value) {template="DISPATCH_RigVMDispatch_ArrayGetAtIndex(in Array,in Index,out Element)"}.Element
   )
   n4 = RigVMDispatchNode(Index=1, Array=RigVMVariableNode() {variable=OriginalThumbs Index}.Value) {template="DISPATCH_RigVMDispatch_ArrayGetAtIndex(in Array,in Index,out Element)"}
   n5 = &RigUnit_GetTransform(Space=LocalSpace, bInitial=False, Item=@n4.Element)
   n6 = &RigVMFunction_MathQuaternionInverse(Value=&RigUnit_GetTransform(Space=LocalSpace, bInitial=true, Item=@n4.Element).Transform.Rotation) {template="Inverse::Execute(in Value,out Result)"}
   > &RigUnit_ModifyTransforms(
    ItemToModify=((Item=(Type=Bone,Name="None"),Transform=(Rotation=(X=0,Y=0,Z=0,W=1),Translation=(X=0,Y=0,Z=0),Scale3D=(X=1,Y=1,Z=1))))
    Weight=1
    WeightMinimum=0
    ItemToModify.0.Transform.Rotation=&RigVMFunction_MathQuaternionMul(
     B=(X=-0.0168,Y=-0.0266,Z=0.0257,W=0.9992)
     A=&RigVMFunction_MathQuaternionMul(A=@n5.Transform.Rotation, B=@n6.Result) {template="Multiply::Execute(in A,in B,out Result)"}.Result
    ) {template="Multiply::Execute(in A,in B,out Result)"}.Result
    ItemToModify.0.Item=RigVMDispatchNode(Index=1, Array=RigVMVariableNode() {variable=CustomOuterThumbs Index}.Value) {template="DISPATCH_RigVMDispatch_ArrayGetAtIndex(in Array,in Index,out Element)"}.Element
   )
   n7 = RigVMDispatchNode(Index=2, Array=RigVMVariableNode() {variable=OriginalThumbs Index}.Value) {template="DISPATCH_RigVMDispatch_ArrayGetAtIndex(in Array,in Index,out Element)"}
   n8 = &RigUnit_GetTransform(Space=LocalSpace, bInitial=False, Item=@n7.Element)
   n9 = &RigVMFunction_MathQuaternionInverse(Value=&RigUnit_GetTransform(Space=LocalSpace, bInitial=true, Item=@n7.Element).Transform.Rotation) {template="Inverse::Execute(in Value,out Result)"}
   > &RigUnit_ModifyTransforms(
    ItemToModify=((Item=(Type=Bone,Name="None"),Transform=(Rotation=(X=0,Y=0,Z=0,W=1),Translation=(X=0,Y=0,Z=0),Scale3D=(X=1,Y=1,Z=1))))
    Weight=1
    WeightMinimum=0
    ItemToModify.0.Transform.Rotation=&RigVMFunction_MathQuaternionMul(
     B=(X=-0.0168,Y=-0.0266,Z=0.0257,W=0.9992)
     A=&RigVMFunction_MathQuaternionMul(A=@n8.Transform.Rotation, B=@n9.Result) {template="Multiply::Execute(in A,in B,out Result)"}.Result
    ) {template="Multiply::Execute(in A,in B,out Result)"}.Result
    ItemToModify.0.Item=RigVMDispatchNode(Index=2, Array=RigVMVariableNode() {variable=CustomOuterThumbs Index}.Value) {template="DISPATCH_RigVMDispatch_ArrayGetAtIndex(in Array,in Index,out Element)"}.Element
   )
   RigUnit_BeginExecution = &RigUnit_BeginExecution()
   RigVMFunction_Sequence = RigVMAggregateNode(ExecuteContext=@RigUnit_BeginExecution.ExecutePin)
   &RigUnit_RotationConstraintLocalSpaceOffset(
    AdvancedSettings=(InterpolationType=Average,RotationOrderForFilter=XZY)
    Child=(Type=Bone,Name="dyn_L_customFrontArm_clavicle")
    Filter=(bX=True,bY=True,bZ=True)
    Parents=((Item=(Type=Bone,Name="clavicle_l"),Weight=1))
    Weight=0.75
    ExecutePin=@RigVMFunction_Sequence.A
   )
   > &RigUnit_RotationConstraintLocalSpaceOffset(
    AdvancedSettings=(InterpolationType=Average,RotationOrderForFilter=XZY)
    Child=(Type=Bone,Name="dyn_L_customFrontArm_upperarm")
    Filter=(bX=True,bY=True,bZ=True)
    Parents=((Item=(Type=Bone,Name="upperarm_l"),Weight=1))
    Weight=1
   )
   > &RigUnit_RotationConstraintLocalSpaceOffset(
    AdvancedSettings=(InterpolationType=Average,RotationOrderForFilter=XZY)
    Child=(Type=Bone,Name="dyn_L_customFrontArm_lowerarm")
    Filter=(bX=True,bY=True,bZ=True)
    Parents=((Item=(Type=Bone,Name="lowerarm_l"),Weight=1))
    Weight=1
   )
   RigUnit_GetTransform = &RigUnit_GetTransform(Item=(Type=Bone,Name="hand_l"))
   > /Script/PBIK.RigUnit_PBIK(
    BoneSettings=()
    Debug=(DrawScale=1)
    Effectors=((Bone="dyn_L_customFrontArm_wrist",Transform=(Rotation=(X=0,Y=0,Z=0,W=1),Translation=(X=0,Y=0,Z=0),Scale3D=(X=1,Y=1,Z=1)),PositionAlpha=1,RotationAlpha=1,StrengthAlpha=1,PullChainAlpha=0,PinRotation=1,ChainDepth=0))
    ExcludedBones=()
    Root=dyn_L_customFrontArm_clavicle
    Settings="(Iterations=20,SubIterations=0,MassMultiplier=1,bAllowStretch=false,RootBehavior=PinToInput,PrePullRootSettings=(RotationAlpha=0,RotationAlphaX=1,RotationAlphaY=1,RotationAlphaZ=1,PositionAlpha=1,PositionAlphaX=1,PositionAlphaY=1,PositionAlphaZ=1),GlobalPu ...(truncated)"
    Effectors.0.Transform=@RigUnit_GetTransform.Transform
   ) {is=struct RigVMUnitNode}
   > &RigUnit_ParentConstraint(
    AdvancedSettings=(InterpolationType=Average,RotationOrderForFilter=XZY)
    Child=(Type=Bone,Name="dyn_L_customFrontArm_wrist")
    Filter=(TranslationFilter=(bX=true,bY=true,bZ=true),RotationFilter=(bX=True,bY=True,bZ=True),ScaleFilter=(bX=True,bY=True,bZ=True))
    Parents=((Item=(Type=Bone,Name="hand_l"),Weight=1))
    Weight=1
   )
   > &RigUnit_ParentConstraint(
    AdvancedSettings=(InterpolationType=Average,RotationOrderForFilter=XZY)
    Child=(Type=Bone,Name="dyn_L_customBackArm_clavicle")
    Parents=((Item=(Type=Bone,Name="dyn_L_customFrontArm_clavicle"),Weight=1))
    Weight=1
   )
   > &RigUnit_ParentConstraint(
    AdvancedSettings=(InterpolationType=Average,RotationOrderForFilter=XZY)
    Child=(Type=Bone,Name="dyn_L_customBackArm_upperarm")
    Parents=((Item=(Type=Bone,Name="dyn_L_customFrontArm_upperarm"),Weight=1))
    Weight=1
   )
   > &RigUnit_ParentConstraint(
    AdvancedSettings=(InterpolationType=Average,RotationOrderForFilter=XZY)
    Child=(Type=Bone,Name="dyn_L_customBackArm_lowerarm")
    Parents=((Item=(Type=Bone,Name="dyn_L_customFrontArm_lowerarm"),Weight=1))
    Weight=1
   )
   > &RigUnit_ParentConstraint(
    AdvancedSettings=(InterpolationType=Average,RotationOrderForFilter=XZY)
    Child=(Type=Bone,Name="dyn_L_customBackArm_wrist")
    Parents=((Item=(Type=Bone,Name="dyn_L_customFrontArm_wrist"),Weight=1))
    Weight=1
   )
   > &RigUnit_ParentConstraint(
    AdvancedSettings=(InterpolationType=Average,RotationOrderForFilter=XZY)
    Child=(Type=Bone,Name="dyn_L_custom_thumb_1")
    Filter=(TranslationFilter=(bX=false,bY=false,bZ=false),RotationFilter=(bX=True,bY=True,bZ=True),ScaleFilter=(bX=True,bY=True,bZ=True))
    Parents=((Item=(Type=Bone,Name="thumb_01_l"),Weight=1))
    Weight=1
   )
   > &RigUnit_ParentConstraint(
    AdvancedSettings=(InterpolationType=Average,RotationOrderForFilter=XZY)
    Child=(Type=Bone,Name="dyn_L_custom_thumb_2")
    Filter=(TranslationFilter=(bX=false,bY=false,bZ=false),RotationFilter=(bX=True,bY=True,bZ=True),ScaleFilter=(bX=True,bY=True,bZ=True))
    Parents=((Item=(Type=Bone,Name="thumb_02_l"),Weight=1))
    Weight=1
   )
   > &RigUnit_ParentConstraint(
    AdvancedSettings=(InterpolationType=Average,RotationOrderForFilter=XZY)
    Child=(Type=Bone,Name="dyn_L_custom_thumb_3")
    Filter=(TranslationFilter=(bX=false,bY=false,bZ=false),RotationFilter=(bX=True,bY=True,bZ=True),ScaleFilter=(bX=True,bY=True,bZ=True))
    Parents=((Item=(Type=Bone,Name="thumb_03_l"),Weight=1))
    Weight=1
   )
   > &RigUnit_ParentConstraint(
    AdvancedSettings=(InterpolationType=Average,RotationOrderForFilter=XZY)
    Child=(Type=Bone,Name="dyn_L_custom_index_1")
    Filter=(TranslationFilter=(bX=false,bY=false,bZ=false),RotationFilter=(bX=True,bY=True,bZ=True),ScaleFilter=(bX=True,bY=True,bZ=True))
    Parents=((Item=(Type=Bone,Name="index_01_l"),Weight=1),(Item=(Type=Bone,Name="middle_01_l"),Weight=0))
    Weight=1
   )
   > &RigUnit_ParentConstraint(
    AdvancedSettings=(InterpolationType=Average,RotationOrderForFilter=XZY)
    Child=(Type=Bone,Name="dyn_L_custom_index_2")
    Filter=(TranslationFilter=(bX=false,bY=false,bZ=false),RotationFilter=(bX=True,bY=True,bZ=True),ScaleFilter=(bX=True,bY=True,bZ=True))
    Parents=((Item=(Type=Bone,Name="index_02_l"),Weight=1),(Item=(Type=Bone,Name="middle_02_l"),Weight=0))
    Weight=1
   )
   > &RigUnit_ParentConstraint(
    AdvancedSettings=(InterpolationType=Average,RotationOrderForFilter=XZY)
    Child=(Type=Bone,Name="dyn_L_custom_index_3")
    Filter=(TranslationFilter=(bX=false,bY=false,bZ=false),RotationFilter=(bX=True,bY=True,bZ=True),ScaleFilter=(bX=True,bY=True,bZ=True))
    Parents=((Item=(Type=Bone,Name="index_03_l"),Weight=1),(Item=(Type=Bone,Name="middle_03_l"),Weight=0))
    Weight=1
   )
   > &RigUnit_ParentConstraint(
    AdvancedSettings=(InterpolationType=Average,RotationOrderForFilter=XZY)
    Child=(Type=Bone,Name="dyn_L_custom_ring_1")
    Filter=(TranslationFilter=(bX=false,bY=false,bZ=false),RotationFilter=(bX=True,bY=True,bZ=True),ScaleFilter=(bX=True,bY=True,bZ=True))
    Parents=((Item=(Type=Bone,Name="ring_01_l"),Weight=1),(Item=(Type=Bone,Name="middle_01_l"),Weight=0))
    Weight=1
   )
   > &RigUnit_ParentConstraint(
    AdvancedSettings=(InterpolationType=Average,RotationOrderForFilter=XZY)
    Child=(Type=Bone,Name="dyn_L_custom_ring_2")
    Filter=(TranslationFilter=(bX=false,bY=false,bZ=false),RotationFilter=(bX=True,bY=True,bZ=True),ScaleFilter=(bX=True,bY=True,bZ=True))
    Parents=((Item=(Type=Bone,Name="ring_02_l"),Weight=1),(Item=(Type=Bone,Name="middle_02_l"),Weight=0))
    Weight=1
   )
   > &RigUnit_ParentConstraint(
    AdvancedSettings=(InterpolationType=Average,RotationOrderForFilter=XZY)
    Child=(Type=Bone,Name="dyn_L_custom_ring_3")
    Filter=(TranslationFilter=(bX=false,bY=false,bZ=false),RotationFilter=(bX=True,bY=True,bZ=True),ScaleFilter=(bX=True,bY=True,bZ=True))
    Parents=((Item=(Type=Bone,Name="ring_03_l"),Weight=1),(Item=(Type=Bone,Name="middle_03_l"),Weight=0))
    Weight=1
   )
   > &RigUnit_ParentConstraint(
    AdvancedSettings=(InterpolationType=Average,RotationOrderForFilter=XZY)
    Child=(Type=Bone,Name="dyn_L_custom_extraThumb_1")
    Filter=(TranslationFilter=(bX=false,bY=false,bZ=false),RotationFilter=(bX=True,bY=True,bZ=True),ScaleFilter=(bX=True,bY=True,bZ=True))
    Parents=((Item=(Type=Bone,Name="pinky_metacarpal_l"),Weight=1))
    Weight=1
   )
   > &RigUnit_ParentConstraint(
    AdvancedSettings=(InterpolationType=Average,RotationOrderForFilter=XZY)
    Child=(Type=Bone,Name="dyn_L_custom_extraThumb_2")
    Filter=(TranslationFilter=(bX=false,bY=false,bZ=false),RotationFilter=(bX=True,bY=True,bZ=True),ScaleFilter=(bX=True,bY=True,bZ=True))
    Parents=((Item=(Type=Bone,Name="pinky_01_l"),Weight=1))
    Weight=1
   )
   > &RigUnit_ParentConstraint(
    AdvancedSettings=(InterpolationType=Average,RotationOrderForFilter=XZY)
    Child=(Type=Bone,Name="dyn_L_custom_extraThumb_3")
    Filter=(TranslationFilter=(bX=false,bY=false,bZ=false),RotationFilter=(bX=True,bY=True,bZ=True),ScaleFilter=(bX=True,bY=True,bZ=True))
    Parents=((Item=(Type=Bone,Name="pinky_02_l"),Weight=1))
    Weight=1
   )
   > &RigUnit_RotationConstraintLocalSpaceOffset(
    AdvancedSettings=(InterpolationType=Average,RotationOrderForFilter=XZY)
    Child=(Type=Bone,Name="dyn_R_customFrontArm_clavicle")
    Filter=(bX=True,bY=True,bZ=True)
    Parents=((Item=(Type=Bone,Name="clavicle_r"),Weight=1))
    Weight=0.75
   )
   > &RigUnit_RotationConstraintLocalSpaceOffset(
    AdvancedSettings=(InterpolationType=Average,RotationOrderForFilter=XZY)
    Child=(Type=Bone,Name="dyn_R_customFrontArm_upperarm")
    Filter=(bX=True,bY=True,bZ=True)
    Parents=((Item=(Type=Bone,Name="upperarm_r"),Weight=1))
    Weight=1
   )
   > &RigUnit_RotationConstraintLocalSpaceOffset(
    AdvancedSettings=(InterpolationType=Average,RotationOrderForFilter=XZY)
    Child=(Type=Bone,Name="dyn_R_customFrontArm_lowerarm")
    Filter=(bX=True,bY=True,bZ=True)
    Parents=((Item=(Type=Bone,Name="lowerarm_r"),Weight=1))
    Weight=1
   )
   > &RigUnit_PBIK(
    BoneSettings=()
    Debug=(DrawScale=1)
    Effectors=((Bone="dyn_R_customFrontArm_wrist",Transform=(Rotation=(X=0,Y=0,Z=0,W=1),Translation=(X=0,Y=0,Z=0),Scale3D=(X=1,Y=1,Z=1)),PositionAlpha=1,RotationAlpha=1,StrengthAlpha=1,PullChainAlpha=1,PinRotation=1,ChainDepth=0))
    ExcludedBones=()
    Root=dyn_R_customFrontArm_clavicle
    Settings="(Iterations=20,SubIterations=0,MassMultiplier=1,bAllowStretch=False,RootBehavior=PinToInput,PrePullRootSettings=(RotationAlpha=0,RotationAlphaX=1,RotationAlphaY=1,RotationAlphaZ=1,PositionAlpha=1,PositionAlphaX=1,PositionAlphaY=1,PositionAlphaZ=1),GlobalPu ...(truncated)"
    Effectors.0.Transform=&RigUnit_GetTransform(Item=(Type=Bone,Name="hand_r")).Transform
   )
   > &RigUnit_ParentConstraint(
    AdvancedSettings=(InterpolationType=Average,RotationOrderForFilter=XZY)
    Child=(Type=Bone,Name="dyn_R_customFrontArm_wrist")
    Filter=(TranslationFilter=(bX=true,bY=true,bZ=true),RotationFilter=(bX=True,bY=True,bZ=True),ScaleFilter=(bX=True,bY=True,bZ=True))
    Parents=((Item=(Type=Bone,Name="hand_r"),Weight=1))
    Weight=1
   )
   > &RigUnit_ParentConstraint(
    AdvancedSettings=(InterpolationType=Average,RotationOrderForFilter=XZY)
    Child=(Type=Bone,Name="dyn_R_customBackArm_clavicle")
    Parents=((Item=(Type=Bone,Name="dyn_R_customFrontArm_clavicle"),Weight=1))
    Weight=1
   )
   > &RigUnit_ParentConstraint(
    AdvancedSettings=(InterpolationType=Average,RotationOrderForFilter=XZY)
    Child=(Type=Bone,Name="dyn_R_customBackArm_upperarm")
    Parents=((Item=(Type=Bone,Name="dyn_R_customFrontArm_upperarm"),Weight=1))
    Weight=1
   )
   > &RigUnit_ParentConstraint(
    AdvancedSettings=(InterpolationType=Average,RotationOrderForFilter=XZY)
    Child=(Type=Bone,Name="dyn_R_customBackArm_lowerarm")
    Parents=((Item=(Type=Bone,Name="dyn_R_customFrontArm_lowerarm"),Weight=1))
    Weight=1
   )
   > &RigUnit_ParentConstraint(
    AdvancedSettings=(InterpolationType=Average,RotationOrderForFilter=XZY)
    Child=(Type=Bone,Name="dyn_R_customBackArm_wrist")
    Parents=((Item=(Type=Bone,Name="dyn_R_customFrontArm_wrist"),Weight=1))
    Weight=1
   )
   > &RigUnit_ParentConstraint(
    AdvancedSettings=(InterpolationType=Average,RotationOrderForFilter=XZY)
    Child=(Type=Bone,Name="dyn_R_custom_thumb_1")
    Filter=(TranslationFilter=(bX=false,bY=false,bZ=false),RotationFilter=(bX=True,bY=True,bZ=True),ScaleFilter=(bX=True,bY=True,bZ=True))
    Parents=((Item=(Type=Bone,Name="thumb_01_r"),Weight=1))
    Weight=1
   )
   > &RigUnit_ParentConstraint(
    AdvancedSettings=(InterpolationType=Average,RotationOrderForFilter=XZY)
    Child=(Type=Bone,Name="dyn_R_custom_thumb_2")
    Filter=(TranslationFilter=(bX=false,bY=false,bZ=false),RotationFilter=(bX=True,bY=True,bZ=True),ScaleFilter=(bX=True,bY=True,bZ=True))
    Parents=((Item=(Type=Bone,Name="thumb_02_r"),Weight=1))
    Weight=1
   )
   > &RigUnit_ParentConstraint(
    AdvancedSettings=(InterpolationType=Average,RotationOrderForFilter=XZY)
    Child=(Type=Bone,Name="dyn_R_custom_thumb_3")
    Filter=(TranslationFilter=(bX=false,bY=false,bZ=false),RotationFilter=(bX=True,bY=True,bZ=True),ScaleFilter=(bX=True,bY=True,bZ=True))
    Parents=((Item=(Type=Bone,Name="thumb_03_r"),Weight=1))
    Weight=1
   )
   > &RigUnit_ParentConstraint(
    AdvancedSettings=(InterpolationType=Average,RotationOrderForFilter=XZY)
    Child=(Type=Bone,Name="dyn_R_custom_index_1")
    Filter=(TranslationFilter=(bX=false,bY=false,bZ=false),RotationFilter=(bX=True,bY=True,bZ=True),ScaleFilter=(bX=True,bY=True,bZ=True))
    Parents=((Item=(Type=Bone,Name="index_01_r"),Weight=1),(Item=(Type=Bone,Name="middle_01_r"),Weight=0))
    Weight=1
   )
   > &RigUnit_ParentConstraint(
    AdvancedSettings=(InterpolationType=Average,RotationOrderForFilter=XZY)
    Child=(Type=Bone,Name="dyn_R_custom_index_2")
    Filter=(TranslationFilter=(bX=false,bY=false,bZ=false),RotationFilter=(bX=True,bY=True,bZ=True),ScaleFilter=(bX=True,bY=True,bZ=True))
    Parents=((Item=(Type=Bone,Name="index_02_r"),Weight=1),(Item=(Type=Bone,Name="middle_02_r"),Weight=0))
    Weight=1
   )
   > &RigUnit_ParentConstraint(
    AdvancedSettings=(InterpolationType=Average,RotationOrderForFilter=XZY)
    Child=(Type=Bone,Name="dyn_R_custom_index_3")
    Filter=(TranslationFilter=(bX=false,bY=false,bZ=false),RotationFilter=(bX=True,bY=True,bZ=True),ScaleFilter=(bX=True,bY=True,bZ=True))
    Parents=((Item=(Type=Bone,Name="index_03_r"),Weight=1),(Item=(Type=Bone,Name="middle_03_r"),Weight=0))
    Weight=1
   )
   > &RigUnit_ParentConstraint(
    AdvancedSettings=(InterpolationType=Average,RotationOrderForFilter=XZY)
    Child=(Type=Bone,Name="dyn_R_custom_ring_1")
    Filter=(TranslationFilter=(bX=false,bY=false,bZ=false),RotationFilter=(bX=True,bY=True,bZ=True),ScaleFilter=(bX=True,bY=True,bZ=True))
    Parents=((Item=(Type=Bone,Name="ring_01_r"),Weight=1),(Item=(Type=Bone,Name="middle_01_r"),Weight=0))
    Weight=1
   )
   > &RigUnit_ParentConstraint(
    AdvancedSettings=(InterpolationType=Average,RotationOrderForFilter=XZY)
    Child=(Type=Bone,Name="dyn_R_custom_ring_2")
    Filter=(TranslationFilter=(bX=false,bY=false,bZ=false),RotationFilter=(bX=True,bY=True,bZ=True),ScaleFilter=(bX=True,bY=True,bZ=True))
    Parents=((Item=(Type=Bone,Name="ring_02_r"),Weight=1),(Item=(Type=Bone,Name="middle_02_r"),Weight=0))
    Weight=1
   )
   > &RigUnit_ParentConstraint(
    AdvancedSettings=(InterpolationType=Average,RotationOrderForFilter=XZY)
    Child=(Type=Bone,Name="dyn_R_custom_ring_3")
    Filter=(TranslationFilter=(bX=false,bY=false,bZ=false),RotationFilter=(bX=True,bY=True,bZ=True),ScaleFilter=(bX=True,bY=True,bZ=True))
    Parents=((Item=(Type=Bone,Name="ring_03_r"),Weight=1),(Item=(Type=Bone,Name="middle_03_r"),Weight=0))
    Weight=1
   )
   > &RigUnit_ParentConstraint(
    AdvancedSettings=(InterpolationType=Average,RotationOrderForFilter=XZY)
    Child=(Type=Bone,Name="dyn_R_custom_extraThumb_1")
    Filter=(TranslationFilter=(bX=false,bY=false,bZ=false),RotationFilter=(bX=True,bY=True,bZ=True),ScaleFilter=(bX=True,bY=True,bZ=True))
    Parents=((Item=(Type=Bone,Name="pinky_metacarpal_r"),Weight=1))
    Weight=0.7
   )
   > &RigUnit_ParentConstraint(
    AdvancedSettings=(InterpolationType=Average,RotationOrderForFilter=XZY)
    Child=(Type=Bone,Name="dyn_R_custom_extraThumb_2")
    Filter=(TranslationFilter=(bX=false,bY=false,bZ=false),RotationFilter=(bX=True,bY=True,bZ=True),ScaleFilter=(bX=True,bY=True,bZ=True))
    Parents=((Item=(Type=Bone,Name="pinky_01_r"),Weight=1))
    Weight=0.7
   )
   > &RigUnit_ParentConstraint(
    AdvancedSettings=(InterpolationType=Average,RotationOrderForFilter=XZY)
    Child=(Type=Bone,Name="dyn_R_custom_extraThumb_3")
    Filter=(TranslationFilter=(bX=false,bY=false,bZ=false),RotationFilter=(bX=True,bY=True,bZ=True),ScaleFilter=(bX=True,bY=True,bZ=True))
    Parents=((Item=(Type=Bone,Name="pinky_02_r"),Weight=1))
    Weight=0.7
   )
   n10 = &RigUnit_GetRelativeTransformForItem(Child=(Type=Bone,Name="upperarm_l"), Parent=(Type=Bone,Name="clavicle_l"), bChildInitial=False, bParentInitial=False)
   > &RigUnit_ModifyTransforms(
    ItemToModify=((Item=(Type=Bone,Name="dyn_L_shoulderPlate_main"),Transform=(Rotation=(X=-0.4988,Y=-0.0349,Z=0.0604,W=0.8639),Translation=(X=0,Y=0,Z=0),Scale3D=(X=1,Y=1,Z=1))))
    WeightMinimum=0
    Weight=&RigVMFunction_MathDoubleRemap(
     SourceMaximum=60
     SourceMinimum=40
     TargetMaximum=1
     TargetMinimum=0
     Value=&RigVMFunction_MathQuaternionToEuler(RotationOrder=ZXY, Value=@n10.RelativeTransform.Rotation).Result.X
    ).Result
   )
   n11 = &RigUnit_GetRelativeTransformForItem(Child=(Type=Bone,Name="upperarm_r"), Parent=(Type=Bone,Name="clavicle_r"), bChildInitial=False, bParentInitial=False)
   > &RigUnit_ModifyTransforms(
    ItemToModify=((Item=(Type=Bone,Name="dyn_R_shoulderPlate_main"),Transform=(Rotation=(X=-0.4988,Y=-0.0349,Z=0.0604,W=0.8639),Translation=(X=0,Y=0,Z=0),Scale3D=(X=1,Y=1,Z=1))))
    WeightMinimum=0
    Weight=&RigVMFunction_MathDoubleRemap(
     SourceMaximum=60
     SourceMinimum=40
     TargetMaximum=1
     TargetMinimum=0
     Value=&RigVMFunction_MathQuaternionToEuler(RotationOrder=ZXY, Value=@n11.RelativeTransform.Rotation).Result.X
    ).Result
   )
   n12 = &RigUnit_GetRelativeTransformForItem(Child=(Type=Bone,Name="lowerarm_l"), Parent=(Type=Bone,Name="upperarm_l"), bChildInitial=False, bParentInitial=False)
   > &RigUnit_ModifyTransforms(
    ItemToModify="((Item=(Type=Bone,Name=\"dyn_L_customBackArm_wireAdjuster\"),Transform=(Rotation=(X=0,Y=0,Z=0,W=1),Translation=(X=-4,Y=0,Z=0),Scale3D=(X=1.2,Y=1,Z=1))),(Item=(Type=Bone,Name=\"dyn_L_customFrontArm_wireAdjuster\"),Transform=(Rotation=(X=0,Y=0,Z=0,W=1),Translati ...(truncated)"
    WeightMinimum=0
    Weight=&RigVMFunction_MathDoubleRemap(
     SourceMaximum=85
     SourceMinimum=50
     TargetMaximum=1
     TargetMinimum=0
     Value=&RigVMFunction_MathQuaternionToEuler(RotationOrder=XYZ, Value=@n12.RelativeTransform.Rotation).Result.Z
    ).Result
   )
   n13 = &RigUnit_GetRelativeTransformForItem(Child=(Type=Bone,Name="lowerarm_r"), Parent=(Type=Bone,Name="upperarm_r"), bChildInitial=False, bParentInitial=False)
   > &RigUnit_ModifyTransforms(
    ItemToModify="((Item=(Type=Bone,Name=\"dyn_R_customBackArm_wireAdjuster\"),Transform=(Rotation=(X=0,Y=0,Z=0,W=1),Translation=(X=4,Y=0,Z=0),Scale3D=(X=1.2,Y=1,Z=1))),(Item=(Type=Bone,Name=\"dyn_R_customFrontArm_wireAdjuster\"),Transform=(Rotation=(X=0,Y=0,Z=0,W=1),Translatio ...(truncated)"
    WeightMinimum=0
    Weight=&RigVMFunction_MathDoubleRemap(
     SourceMaximum=85
     SourceMinimum=50
     TargetMaximum=1
     TargetMinimum=0
     Value=&RigVMFunction_MathQuaternionToEuler(RotationOrder=XYZ, Value=@n13.RelativeTransform.Rotation).Result.Z
    ).Result
   )
   > &RigUnit_RotationConstraintLocalSpaceOffset(
    AdvancedSettings=(InterpolationType=Average,RotationOrderForFilter=XZY)
    Child=(Type=Bone,Name="dyn_L_customFrontArm_clavicle_adjuster")
    Filter=(bX=True,bY=True,bZ=True)
    Parents=((Item=(Type=Bone,Name="dyn_L_customFrontArm_clavicle"),Weight=0.5),(Item=(Type=Bone,Name="dyn_L_customFrontArm_upperarm"),Weight=0.5))
    Weight=1
   )
   > &RigUnit_RotationConstraintLocalSpaceOffset(
    AdvancedSettings=(InterpolationType=Average,RotationOrderForFilter=XZY)
    Child=(Type=Bone,Name="dyn_L_customBackArm_clavicle_adjuster")
    Filter=(bX=True,bY=True,bZ=True)
    Parents=((Item=(Type=Bone,Name="dyn_L_customBackArm_clavicle"),Weight=0.5),(Item=(Type=Bone,Name="dyn_L_customBackArm_upperarm"),Weight=0.5))
    Weight=1
   )
   > &RigUnit_RotationConstraintLocalSpaceOffset(
    AdvancedSettings=(InterpolationType=Average,RotationOrderForFilter=XZY)
    Child=(Type=Bone,Name="dyn_R_customFrontArm_clavicle_adjuster")
    Filter=(bX=True,bY=True,bZ=True)
    Parents=((Item=(Type=Bone,Name="dyn_R_customFrontArm_clavicle"),Weight=0.5),(Item=(Type=Bone,Name="dyn_R_customFrontArm_upperarm"),Weight=0.5))
    Weight=1
   )
   > &RigUnit_RotationConstraintLocalSpaceOffset(
    AdvancedSettings=(InterpolationType=Average,RotationOrderForFilter=XZY)
    Child=(Type=Bone,Name="dyn_R_customBackArm_clavicle_adjuster")
    Filter=(bX=True,bY=True,bZ=True)
    Parents=((Item=(Type=Bone,Name="dyn_R_customBackArm_clavicle"),Weight=0.5),(Item=(Type=Bone,Name="dyn_R_customBackArm_upperarm"),Weight=0.5))
    Weight=1
   )
   n14 = &RigUnit_GetRelativeTransformForItem(Child=(Type=Bone,Name="upperarm_l"), Parent=(Type=Bone,Name="clavicle_l"), bChildInitial=False, bParentInitial=False)
   > &RigUnit_ModifyTransforms(
    ItemToModify=((Item=(Type=Bone,Name="dyn_L_customBackArm_clavicle_adjuster"),Transform=(Rotation=(X=-0.4226,Y=0,Z=0,W=0.9063),Translation=(X=0,Y=-2,Z=0),Scale3D=(X=1,Y=1,Z=1))))
    WeightMinimum=0
    Weight=&RigVMFunction_MathDoubleRemap(
     SourceMaximum=90
     SourceMinimum=50
     TargetMaximum=1
     TargetMinimum=0
     Value=&RigVMFunction_MathQuaternionToEuler(RotationOrder=ZXY, Value=@n14.RelativeTransform.Rotation).Result.Z
    ).Result
   )
   n15 = &RigUnit_GetRelativeTransformForItem(Child=(Type=Bone,Name="upperarm_r"), Parent=(Type=Bone,Name="clavicle_r"), bChildInitial=False, bParentInitial=False)
   > &RigUnit_ModifyTransforms(
    ItemToModify=((Item=(Type=Bone,Name="dyn_R_customBackArm_clavicle_adjuster"),Transform=(Rotation=(X=-0.4226,Y=0,Z=0,W=0.9063),Translation=(X=0,Y=2,Z=0),Scale3D=(X=1,Y=1,Z=1))))
    WeightMinimum=0
    Weight=&RigVMFunction_MathDoubleRemap(
     SourceMaximum=90
     SourceMinimum=50
     TargetMaximum=1
     TargetMinimum=0
     Value=&RigVMFunction_MathQuaternionToEuler(RotationOrder=ZXY, Value=@n15.RelativeTransform.Rotation).Result.Z
    ).Result
   )
   n16 = &RigVMFunction_Sequence(ExecuteContext=@RigVMFunction_Sequence.B)
   RigVMFunction_ControlFlowBranch = &RigVMFunction_ControlFlowBranch(ExecuteContext=@n16.A, Condition=RigVMVariableNode() {variable=shoesEquipped}.Value)
   Multiply = &RigVMFunction_MathTransformMul(
    B=&RigUnit_GetTransform(Item=(Type=Bone,Name="ball_l"), bInitial=false).Transform
    A=&RigVMFunction_MathTransformMakeRelative(
     Parent=&RigUnit_GetTransform(Item=(Type=Bone,Name="ball_l"), bInitial=true).Transform
     Global=&RigUnit_GetTransform(Item=(Type=Bone,Name="dyn_L_custom_ball"), bInitial=true).Transform
    ).Local
   ) {template="Multiply::Execute(in A,in B,out Result)"}
   /Script/ControlRig.RigUnit_FABRIKItemArray(
    Items=((Type=Bone,Name="dyn_L_custom_calf"),(Type=Bone,Name="dyn_L_custom_ball"))
    MaxIterations=10
    Precision=1
    Weight=1
    bPropagateToChildren=True
    bSetEffectorTransform=true
    EffectorTransform=@Multiply.Result
    ExecutePin=@RigVMFunction_ControlFlowBranch.False
   ) {is=struct RigVMUnitNode}
   > &RigUnit_ModifyTransforms(
    ItemToModify="((Item=(Type=Bone,Name=\"thigh_l\"),Transform=(Rotation=(X=0,Y=0,Z=0.0436,W=0.999),Translation=(X=0,Y=0,Z=0),Scale3D=(X=1,Y=1,Z=1))),(Item=(Type=Bone,Name=\"dyn_L_custom_calf\"),Transform=(Rotation=(X=0,Y=0,Z=-0.0436,W=0.999),Translation=(X=0,Y=0,Z=0),Scale3D= ...(truncated)"
    Weight=1
    WeightMinimum=0
   )
   &RigUnit_ParentConstraint(
    AdvancedSettings=(InterpolationType=Average,RotationOrderForFilter=XZY)
    Child=(Type=Bone,Name="thigh_l_IK")
    Filter=(TranslationFilter=(bX=True,bY=True,bZ=True),RotationFilter=(bX=true,bY=true,bZ=true),ScaleFilter=(bX=True,bY=True,bZ=True))
    Parents=((Item=(Type=Bone,Name="thigh_l"),Weight=1))
    Weight=1
    ExecutePin=@n16.B
   )
   > &RigUnit_ParentConstraint(
    AdvancedSettings=(InterpolationType=Average,RotationOrderForFilter=XZY)
    Child=(Type=Bone,Name="dyn_L_custom_ball_corrective")
    Filter=(TranslationFilter=(bX=True,bY=True,bZ=True),RotationFilter=(bX=false,bY=false,bZ=false),ScaleFilter=(bX=True,bY=True,bZ=True))
    Parents=((Item=(Type=Bone,Name="dyn_L_custom_foot"),Weight=1))
    Weight=1
   )
   > &RigUnit_ParentConstraint(
    AdvancedSettings=(InterpolationType=Average,RotationOrderForFilter=XZY)
    Child=(Type=Bone,Name="dyn_L_custom_ball_corrective")
    Filter=(TranslationFilter=(bX=false,bY=false,bZ=false),RotationFilter=(bX=true,bY=true,bZ=true),ScaleFilter=(bX=True,bY=True,bZ=True))
    Parents=((Item=(Type=Bone,Name="ball_l"),Weight=1))
    Weight=1
   )
   n17 = &RigVMFunction_Sequence(ExecuteContext=@RigVMFunction_Sequence.C)
   n18 = &RigVMFunction_ControlFlowBranch(ExecuteContext=@n17.A, Condition=RigVMVariableNode() {variable=shoesEquipped}.Value)
   n19 = &RigUnit_GetTransform(Item=(Type=Bone,Name="ball_r"), bInitial=true)
   n20 = &RigUnit_GetTransform(Item=(Type=Bone,Name="dyn_R_custom_ball"), bInitial=true)
   &RigUnit_FABRIKItemArray(
    Items=((Type=Bone,Name="dyn_R_custom_calf"),(Type=Bone,Name="dyn_R_custom_ball"))
    MaxIterations=10
    Precision=1
    Weight=1
    bPropagateToChildren=True
    bSetEffectorTransform=true
    EffectorTransform=&RigVMFunction_MathTransformMul(
     B=&RigUnit_GetTransform(Item=(Type=Bone,Name="ball_r"), bInitial=false).Transform
     A=&RigVMFunction_MathTransformMakeRelative(Parent=@n19.Transform, Global=@n20.Transform).Local
    ) {template="Multiply::Execute(in A,in B,out Result)"}.Result
    ExecutePin=@n18.False
   )
   > &RigUnit_ModifyTransforms(
    ItemToModify="((Item=(Type=Bone,Name=\"thigh_r\"),Transform=(Rotation=(X=0,Y=0,Z=0.0436,W=0.999),Translation=(X=0,Y=0,Z=0),Scale3D=(X=1,Y=1,Z=1))),(Item=(Type=Bone,Name=\"dyn_R_custom_calf\"),Transform=(Rotation=(X=0,Y=0,Z=-0.0436,W=0.999),Translation=(X=0,Y=0,Z=0),Scale3D= ...(truncated)"
    Weight=1
    WeightMinimum=0
   )
   &RigUnit_ParentConstraint(
    AdvancedSettings=(InterpolationType=Average,RotationOrderForFilter=XZY)
    Child=(Type=Bone,Name="thigh_r_IK")
    Filter=(TranslationFilter=(bX=True,bY=True,bZ=True),RotationFilter=(bX=true,bY=true,bZ=true),ScaleFilter=(bX=True,bY=True,bZ=True))
    Parents=((Item=(Type=Bone,Name="thigh_r"),Weight=1))
    Weight=1
    ExecutePin=@n17.B
   )
   > &RigUnit_ParentConstraint(
    AdvancedSettings=(InterpolationType=Average,RotationOrderForFilter=XZY)
    Child=(Type=Bone,Name="dyn_R_custom_ball_corrective")
    Filter=(TranslationFilter=(bX=True,bY=True,bZ=True),RotationFilter=(bX=false,bY=false,bZ=false),ScaleFilter=(bX=True,bY=True,bZ=True))
    Parents=((Item=(Type=Bone,Name="dyn_R_custom_foot"),Weight=1))
    Weight=1
   )
   > &RigUnit_ParentConstraint(
    AdvancedSettings=(InterpolationType=Average,RotationOrderForFilter=XZY)
    Child=(Type=Bone,Name="dyn_R_custom_ball_corrective")
    Filter=(TranslationFilter=(bX=false,bY=false,bZ=false),RotationFilter=(bX=true,bY=true,bZ=true),ScaleFilter=(bX=True,bY=True,bZ=True))
    Parents=((Item=(Type=Bone,Name="ball_r"),Weight=1))
    Weight=1
   )
  ]}
]
hierarchy [
 root {loc=<0 0 0>, rot=<0 0 0 1>}
 pelvis {parent=root, loc=<0 -0.7071 93.1191>, rot=<0.0127 0.707 -0.0127 -0.707>}
 spine_01 {parent=pelvis, loc=<6.7494 -0.6388 0>, rot=<0 0 0.034 -0.9994>}
 spine_02 {parent=spine_01, loc=<7.252 -0.035 0>, rot=<0 0 0.0238 -0.9997>}
 spine_03 {parent=spine_02, loc=<6.0106 0 0>, rot=<0 0 0.0149 -0.9999>}
 spine_04 {parent=spine_03, loc=<8.1023 0 0>, rot=<0 0 -0.0132 -0.9999>}
 spine_05 {parent=spine_04, loc=<10.0142 0 0>, rot=<0 0 0 1>}
 attach_backpack {parent=spine_05, loc=<-0.0827 -13.4733 0>, rot=<-0.0082 -0.7071 0.0082 -0.7071>}
 neck_01 {parent=spine_05, loc=<11.3133 0.1867 0>, rot=<0 0 -0.2017 -0.9794>}
 neck_02 {parent=neck_01, loc=<4.5627 0 0>, rot=<0 0 0.0504 -0.9987>}
 head {parent=neck_02, loc=<5.3914 0 0>, rot=<0 0 0.115 -0.9934>}
 clavicle_l {parent=spine_05, loc=<4.3579 -0.1336 -2.3223>, rot=<0.0476 -0.7289 -0.01 -0.6829>}
 upperarm_l {parent=clavicle_l, loc=<15.5111 0 0>, rot=<0.0226 -0.3234 -0.0088 -0.9459>}
 lowerarm_l {parent=upperarm_l, loc=<26.6514 0 0>, rot=<0.0391 -0.0099 -0.2453 -0.9686>}
 hand_l {parent=lowerarm_l, loc=<22.3491 0 0>, rot=<-0.7081 -0.0507 0.0975 -0.6975>}
 hand_attach_l {parent=hand_l, loc=<8.5 -3 0>, rot=<0.7071 0.7071 0 0>}
 thumb_01_l {parent=hand_l, loc=<1.3118 -0.3191 -2.459>, rot=<0.6546 -0.4042 0.0664 -0.6354>}
 thumb_02_l {parent=thumb_01_l, loc=<5.9336 -0.0448 0.0401>, rot=<-0.0037 0.0213 0.298 -0.9543>}
 thumb_03_l {parent=thumb_02_l, loc=<2.2223 0.3682 -0.0907>, rot=<0.0118 -0.013 0.0066 -0.9998>}
 index_metacarpal_l {parent=hand_l, loc=<1.6535 0.0183 -2.96>, rot=<0.0158 -0.0124 -0.0173 -0.9996>}
 index_01_l {parent=index_metacarpal_l, loc=<8.3008 0.0357 -0.0524>, rot=<-0.0044 0.0485 0.0924 -0.9945>}
 index_02_l {parent=index_01_l, loc=<3.3587 -0.3004 0.0452>, rot=<-0.002 0.0027 0.3225 -0.9466>}
 index_03_l {parent=index_02_l, loc=<2.5578 -0.123 -0.0252>, rot=<-0.002 0.0067 0.2558 -0.9667>}
 middle_metacarpal_l {parent=hand_l, loc=<2.1856 -0.1725 -0.5984>, rot=<0.006 0.0168 -0.0329 -0.9993>}
 middle_01_l {parent=middle_metacarpal_l, loc=<7.3275 -0.2447 0.0129>, rot=<-0.0208 0.0042 0.0858 -0.9961>}
 middle_02_l {parent=middle_01_l, loc=<4.0123 0.1401 0.1175>, rot=<-0.0218 0.0258 0.3512 -0.9357>}
 middle_03_l {parent=middle_02_l, loc=<2.6077 0.0534 0.1373>, rot=<0.0082 0.0188 0.1708 -0.9851>}
 ring_metacarpal_l {parent=hand_l, loc=<2.2074 -0.3625 1.3638>, rot=<-0.0284 0.0722 -0.0458 -0.9959>}
 ring_01_l {parent=ring_metacarpal_l, loc=<7.0295 -0.6197 -0.1446>, rot=<-0.0489 -0.0252 0.1892 -0.9804>}
 ring_02_l {parent=ring_01_l, loc=<3.6174 0.4647 -0.1027>, rot=<0 0 0.2806 -0.9598>}
 ring_03_l {parent=ring_02_l, loc=<2.4152 0.114 0.0199>, rot=<0 0 0.1189 -0.9929>}
 pinky_metacarpal_l {parent=hand_l, loc=<1.7616 -0.9753 3.3814>, rot=<-0.0397 0.0814 -0.069 -0.9935>}
 pinky_01_l {parent=pinky_metacarpal_l, loc=<7.171 -0.483 -0.1094>, rot=<-0.0654 -0.0035 0.2242 -0.9723>}
 pinky_02_l {parent=pinky_01_l, loc=<2.8274 0.0892 -0.1043>, rot=<0.0037 0.0011 0.2854 -0.9584>}
 pinky_03_l {parent=pinky_02_l, loc=<2.2046 0.0307 -0.0078>, rot=<-0.0002 0.0098 0.1445 -0.9895>}
 weapon_l {parent=hand_l, loc=<-0.1974 -3.4111 0.3811>, rot=<0 0 0 1>}
 lowerarm_twist_02_l {parent=lowerarm_l, loc=<9.6199 0 0>, rot=<0 0 0 -1>}
 lowerarm_twist_01_l {parent=lowerarm_l, loc=<17.8549 0 0>, rot=<0 0 0 -1>}
 upperarm_twist_01_l {parent=upperarm_l, loc=<0.0041 0 0>, rot=<0 0 0 -1>}
 upperarm_twist_02_l {parent=upperarm_l, loc=<13.7941 0 0>, rot=<0 0 0 -1>}
 attach_cape {parent=spine_05, loc=<9.052 -11.687 0>, rot=<0.0627 -0.7043 -0.0627 -0.7043>}
 clavicle_r {parent=spine_05, loc=<4.3579 -0.1335 2.3223>, rot=<0.7289 0.0476 -0.6829 0.01>}
 upperarm_r {parent=clavicle_r, loc=<-15.5111 0 0>, rot=<0.0226 -0.3234 -0.0088 -0.9459>}
 lowerarm_r {parent=upperarm_r, loc=<-26.6514 0 0>, rot=<0.0391 -0.0099 -0.2453 -0.9686>}
 hand_r {parent=lowerarm_r, loc=<-22.3491 0 0>, rot=<-0.7081 -0.0507 0.0975 -0.6975>}
 thumb_01_r {parent=hand_r, loc=<-1.3118 0.3191 2.459>, rot=<0.6546 -0.4042 0.0664 -0.6354>}
 thumb_02_r {parent=thumb_01_r, loc=<-5.9336 0.0448 -0.0401>, rot=<-0.0037 0.0213 0.298 -0.9543>}
 thumb_03_r {parent=thumb_02_r, loc=<-2.2223 -0.3682 0.0907>, rot=<0.0118 -0.013 0.0066 -0.9998>}
 index_metacarpal_r {parent=hand_r, loc=<-1.6535 -0.0183 2.96>, rot=<0.0158 -0.0124 -0.0173 -0.9996>}
 index_01_r {parent=index_metacarpal_r, loc=<-8.3008 -0.0357 0.0524>, rot=<-0.0044 0.0485 0.0924 -0.9945>}
 index_02_r {parent=index_01_r, loc=<-3.3587 0.3004 -0.0452>, rot=<-0.002 0.0027 0.3225 -0.9466>}
 index_03_r {parent=index_02_r, loc=<-2.5578 0.123 0.0252>, rot=<-0.002 0.0067 0.2558 -0.9667>}
 middle_metacarpal_r {parent=hand_r, loc=<-2.1856 0.1725 0.5984>, rot=<0.006 0.0168 -0.0329 -0.9993>}
 middle_01_r {parent=middle_metacarpal_r, loc=<-7.3275 0.2447 -0.0129>, rot=<-0.0208 0.0042 0.0858 -0.9961>}
 middle_02_r {parent=middle_01_r, loc=<-4.0123 -0.1401 -0.1175>, rot=<-0.0218 0.0258 0.3512 -0.9357>}
 middle_03_r {parent=middle_02_r, loc=<-2.6077 -0.0534 -0.1373>, rot=<0.0082 0.0188 0.1708 -0.9851>}
 ring_metacarpal_r {parent=hand_r, loc=<-2.2074 0.3625 -1.3638>, rot=<-0.0284 0.0722 -0.0458 -0.9959>}
 ring_01_r {parent=ring_metacarpal_r, loc=<-7.0295 0.6197 0.1446>, rot=<-0.0489 -0.0252 0.1892 -0.9804>}
 ring_02_r {parent=ring_01_r, loc=<-3.6174 -0.4647 0.1027>, rot=<0 0 0.2806 -0.9598>}
 ring_03_r {parent=ring_02_r, loc=<-2.4152 -0.114 -0.0199>, rot=<0 0 0.1189 -0.9929>}
 pinky_metacarpal_r {parent=hand_r, loc=<-1.7616 0.9753 -3.3814>, rot=<-0.0397 0.0814 -0.069 -0.9935>}
 pinky_01_r {parent=pinky_metacarpal_r, loc=<-7.1709 0.4831 0.1094>, rot=<-0.0654 -0.0035 0.2242 -0.9723>}
 pinky_02_r {parent=pinky_01_r, loc=<-2.8274 -0.0893 0.1043>, rot=<0.0037 0.0011 0.2854 -0.9584>}
 pinky_03_r {parent=pinky_02_r, loc=<-2.2046 -0.0307 0.0078>, rot=<-0.0002 0.0098 0.1445 -0.9895>}
 weapon_r {parent=hand_r, loc=<0.1974 3.4111 -0.3811>, rot=<0 0 0 1>}
 hand_attach_r {parent=hand_r, loc=<-8.5001 2.9999 0>, rot=<0 0 -0.7071 -0.7071>}
 lowerarm_twist_02_r {parent=lowerarm_r, loc=<-9.6199 0 0>, rot=<0 0 0 -1>}
 lowerarm_twist_01_r {parent=lowerarm_r, loc=<-17.8549 0 0>, rot=<0 0 0 -1>}
 upperarm_twist_01_r {parent=upperarm_r, loc=<-0.0041 0 0>, rot=<0 0 0 -1>}
 upperarm_twist_02_r {parent=upperarm_r, loc=<-13.7941 0 0>, rot=<0 0 0 -1>}
 dyn_L_customFrontArm_clavicle {parent=spine_05, loc=<-3.6963 1.4162 -9.5394>, rot=<0.0288 -0.3736 -0.0599 -0.9252>}
 dyn_L_customFrontArm_upperarm {parent=dyn_L_customFrontArm_clavicle, loc=<10.2442 0 0>, rot=<0 -0.6806 0 -0.7327>}
 dyn_L_customFrontArm_lowerarm {parent=dyn_L_customFrontArm_upperarm, loc=<26.397 0 0>, rot=<0.1856 -0.0422 -0.2176 -0.9573>}
 dyn_L_customFrontArm_wrist {parent=dyn_L_customFrontArm_lowerarm, loc=<23.1045 -0.7127 -0.3162>, rot=<-0.7071 0 0 -0.7071>}
 dyn_L_custom_thumb_1 {parent=dyn_L_customFrontArm_wrist, loc=<1.9169 0.5869 -1.1471>, rot=<-0.1438 -0.4135 0.3197 -0.8403>}
 dyn_L_custom_thumb_2 {parent=dyn_L_custom_thumb_1, loc=<4.6371 0 0>, rot=<0 0.3526 0 -0.9358>}
 dyn_L_custom_thumb_3 {parent=dyn_L_custom_thumb_2, loc=<2.5336 0 0>, rot=<0 0.2518 0 -0.9678>}
 dyn_L_custom_index_1 {parent=dyn_L_customFrontArm_wrist, loc=<8.8135 1.0897 -0.6224>, rot=<0.7178 -0.1239 -0.0064 -0.6851>}
 dyn_L_custom_index_2 {parent=dyn_L_custom_index_1, loc=<3.5083 0 0>, rot=<0 -0.2835 0 -0.959>}
 dyn_L_custom_index_3 {parent=dyn_L_custom_index_2, loc=<2.9925 0 0>, rot=<-0.1582 -0.4125 -0.073 -0.8941>}
 dyn_L_customFrontArm_wireAdjuster {parent=dyn_L_customFrontArm_lowerarm, loc=<8.2109 0.0484 0.1462>, rot=<0 0 0 1>}
 dyn_L_customFrontArm_clavicle_adjuster {parent=dyn_L_customFrontArm_clavicle, loc=<4.7949 1.0094 1.168>, rot=<-0.0137 0.0839 -0.0332 -0.9958>}
 dyn_L_customBackArm_clavicle {parent=spine_05, loc=<-4.2849 -2.1385 -9.2883>, rot=<0.0206 -0.3627 -0.0346 -0.9311>}
 dyn_L_customBackArm_upperarm {parent=dyn_L_customBackArm_clavicle, loc=<11.8541 0 0>, rot=<0 -0.695 0 -0.719>}
 dyn_L_customBackArm_lowerarm {parent=dyn_L_customBackArm_upperarm, loc=<28.246 0 0>, rot=<0.1791 -0.0455 -0.2421 -0.9525>}
 dyn_L_customBackArm_wrist {parent=dyn_L_customBackArm_lowerarm, loc=<22.4142 -0.5164 -0.442>, rot=<-0.7071 0 0 -0.7071>}
 dyn_L_custom_ring_1 {parent=dyn_L_customBackArm_wrist, loc=<8.6761 0.0888 1.8777>, rot=<0.5269 0.0272 0.0974 -0.8439>}
 dyn_L_custom_ring_2 {parent=dyn_L_custom_ring_1, loc=<3.2343 0 0>, rot=<0 -0.3309 0 -0.9437>}
 dyn_L_custom_ring_3 {parent=dyn_L_custom_ring_2, loc=<3.1061 0 0>, rot=<0.0274 -0.4048 0.0121 -0.9139>}
 dyn_L_custom_extraThumb_1 {parent=dyn_L_customBackArm_wrist, loc=<1.5302 0.0024 1.4683>, rot=<-0.0058 0.3074 0.2728 -0.9116>}
 dyn_L_custom_extraThumb_2 {parent=dyn_L_custom_extraThumb_1, loc=<4.3712 0 0>, rot=<0 -0.2723 0 -0.9622>}
 dyn_L_custom_extraThumb_3 {parent=dyn_L_custom_extraThumb_2, loc=<3.9198 0 0>, rot=<0 -0.3043 0 -0.9526>}
 dyn_L_customBackArm_wireAdjuster {parent=dyn_L_customBackArm_lowerarm, loc=<7.9802 0.16 -0.3091>, rot=<-0.2106 -0.0207 -0.0053 -0.9773>}
 dyn_L_shoulderPlate_front {parent=dyn_L_customBackArm_upperarm, loc=<-1.1536 8.0221 1.1063>, rot=<0.4473 -0.7362 -0.2605 0.436>}
 dyn_L_shoulderPlate_back {parent=dyn_L_shoulderPlate_front, loc=<11.3198 -10.057 -0.2791>, rot=<0.0894 0.0504 -0.7721 -0.6272>}
 dyn_L_shoulderPlate_main {parent=dyn_L_shoulderPlate_back, loc=<6.2584 4.7544 1.5821>, rot=<-0.3177 -0.7658 0.2649 -0.4925>}
 dyn_L_customBackArm_clavicle_adjuster {parent=dyn_L_customBackArm_clavicle, loc=<4.7972 -1.3501 0.8084>, rot=<-0.0045 0.0709 0.016 -0.9973>}
 dyn_R_customFrontArm_clavicle {parent=spine_05, loc=<-3.6962 -0.1836 9.6423>, rot=<-0.3736 -0.0288 0.9252 -0.0599>}
 dyn_R_customFrontArm_upperarm {parent=dyn_R_customFrontArm_clavicle, loc=<-10.2441 0 0.0002>, rot=<0 -0.6806 0 -0.7327>}
 dyn_R_customFrontArm_lowerarm {parent=dyn_R_customFrontArm_upperarm, loc=<-26.3968 0 -0.0001>, rot=<0.1856 -0.0422 -0.2176 -0.9573>}
 dyn_R_customFrontArm_wrist {parent=dyn_R_customFrontArm_lowerarm, loc=<-23.5231 -0.6802 -0.3579>, rot=<-0.7071 0 0 -0.7071>}
 dyn_R_custom_thumb_1 {parent=dyn_R_customFrontArm_wrist, loc=<-1.9174 -0.5865 1.1471>, rot=<-0.1438 -0.4135 0.3197 -0.8403>}
 dyn_R_custom_thumb_2 {parent=dyn_R_custom_thumb_1, loc=<-4.6368 0 0.0004>, rot=<0 0.3526 0 -0.9358>}
 dyn_R_custom_thumb_3 {parent=dyn_R_custom_thumb_2, loc=<-2.5336 0 -0.0001>, rot=<0 0.2518 0 -0.9678>}
 dyn_R_custom_index_1 {parent=dyn_R_customFrontArm_wrist, loc=<-8.8136 -1.0897 0.6224>, rot=<0.7178 -0.1239 -0.0064 -0.6851>}
 dyn_R_custom_index_2 {parent=dyn_R_custom_index_1, loc=<-3.5084 0 0>, rot=<0 -0.2835 0 -0.959>}
 dyn_R_custom_index_3 {parent=dyn_R_custom_index_2, loc=<-2.9925 0 0>, rot=<-0.1582 -0.4125 -0.073 -0.8941>}
 dyn_R_customFrontArm_wireAdjuster {parent=dyn_R_customFrontArm_lowerarm, loc=<-8.211 -0.0484 -0.1462>, rot=<0 0 0 1>}
 dyn_R_customFrontArm_clavicle_adjuster {parent=dyn_R_customFrontArm_clavicle, loc=<-4.7948 -1.0094 -1.168>, rot=<-0.0137 0.0839 -0.0332 -0.9958>}
 dyn_R_customBackArm_clavicle {parent=spine_05, loc=<-4.2845 -3.6476 8.8057>, rot=<-0.3627 -0.0206 0.9311 -0.0346>}
 dyn_R_customBackArm_upperarm {parent=dyn_R_customBackArm_clavicle, loc=<-11.854 0 0.0001>, rot=<0 -0.695 0 -0.719>}
 dyn_R_customBackArm_lowerarm {parent=dyn_R_customBackArm_upperarm, loc=<-28.2462 0 0.0002>, rot=<0.1791 -0.0455 -0.2421 -0.9525>}
 dyn_R_customBackArm_wrist {parent=dyn_R_customBackArm_lowerarm, loc=<-23.2584 -0.4646 -0.4717>, rot=<-0.7071 0 0 -0.7071>}
 dyn_R_custom_ring_1 {parent=dyn_R_customBackArm_wrist, loc=<-8.6758 -0.0892 -1.8776>, rot=<0.5269 0.0272 0.0974 -0.8439>}
 dyn_R_custom_ring_2 {parent=dyn_R_custom_ring_1, loc=<-3.2344 0 0>, rot=<0 -0.3309 0 -0.9437>}
 dyn_R_custom_ring_3 {parent=dyn_R_custom_ring_2, loc=<-3.1061 0 0>, rot=<0.0274 -0.4048 0.0121 -0.9139>}
 dyn_R_custom_extraThumb_1 {parent=dyn_R_customBackArm_wrist, loc=<-1.5299 -0.0028 -1.4683>, rot=<-0.0058 0.3074 0.2728 -0.9116>}
 dyn_R_custom_extraThumb_2 {parent=dyn_R_custom_extraThumb_1, loc=<-4.3711 0.0001 0>, rot=<0 -0.2723 0 -0.9622>}
 dyn_R_custom_extraThumb_3 {parent=dyn_R_custom_extraThumb_2, loc=<-3.9199 0 0>, rot=<0 -0.3043 0 -0.9526>}
 dyn_R_customBackArm_wireAdjuster {parent=dyn_R_customBackArm_lowerarm, loc=<-7.9801 -0.16 0.309>, rot=<-0.2106 -0.0207 -0.0053 -0.9773>}
 dyn_R_shoulderPlate_front {parent=dyn_R_customBackArm_upperarm, loc=<0.8802 -9.5328 -1.4978>, rot=<0.4473 -0.7362 -0.2605 0.436>}
 dyn_R_shoulderPlate_back {parent=dyn_R_shoulderPlate_front, loc=<-11.3199 10.0567 0.279>, rot=<0.0894 0.0504 -0.7721 -0.6272>}
 dyn_R_shoulderPlate_main {parent=dyn_R_shoulderPlate_back, loc=<-6.2586 -4.7543 -1.5821>, rot=<-0.3177 -0.7658 0.2649 -0.4925>}
 dyn_R_customBackArm_clavicle_adjuster {parent=dyn_R_customBackArm_clavicle, loc=<-4.7966 1.3501 -0.8079>, rot=<-0.0045 0.0709 0.016 -0.9973>}
 thigh_l {parent=pelvis, loc=<-2.5693 -0.0109 -7.7501>, rot=<0.0012 0.0572 0.0155 -0.9982>}
 calf_l {parent=thigh_l, loc=<-39.0484 0 0>, rot=<0.0006 -0.0161 -0.0391 -0.9991>}
 foot_l {parent=calf_l, loc=<-42.561 0.1077 0.4292>, rot=<0.0418 -0.0455 0.0343 -0.9975>}
 ball_l {parent=foot_l, loc=<-6.1599 13.8264 0.6806>, rot=<0.0001 -0.0001 -0.7063 -0.7079>}
 calf_twist_02_l {parent=calf_l, loc=<-18.8064 -0.0967 0.0397>, rot=<0.0237 0.0095 -0.0035 -0.9997>}
 calf_twist_01_l {parent=calf_l, loc=<-35.813 -0.0967 0.0397>, rot=<0.0237 0.0095 -0.0035 -0.9997>}
 thigh_twist_01_l {parent=thigh_l, loc=<-21.6303 0 0>, rot=<0.0013 0 0 -1>}
 dyn_L_custom_calf {parent=thigh_l, loc=<-40.3678 4.4862 0.0384>, rot=<0.0395 -0.0113 -0.1563 -0.9868>}
 dyn_L_custom_foot {parent=dyn_L_custom_calf, loc=<-35.4723 0 0>, rot=<-0.0624 -0.0349 0.4866 -0.8707>}
 dyn_L_custom_ball {parent=dyn_L_custom_foot, loc=<-19.3509 0 0>, rot=<0 0 0.4209 -0.9071>}
 dyn_L_custom_ball_corrective {parent=dyn_L_custom_foot, loc=<-19.3509 0 0>, rot=<0 0 0.4209 -0.9071>}
 thigh_r {parent=pelvis, loc=<-2.5693 -0.0109 7.7501>, rot=<0.0572 -0.0012 0.9982 0.0155>}
 calf_r {parent=thigh_r, loc=<39.0484 0 0>, rot=<0.0006 -0.0161 -0.0391 -0.9991>}
 foot_r {parent=calf_r, loc=<42.561 -0.1077 -0.4292>, rot=<0.0418 -0.0455 0.0343 -0.9975>}
 ball_r {parent=foot_r, loc=<6.1599 -13.8264 -0.6806>, rot=<0.0001 -0.0001 -0.7063 -0.7079>}
 calf_twist_02_r {parent=calf_r, loc=<18.8064 0.0967 -0.0397>, rot=<0.0237 0.0095 -0.0035 -0.9997>}
 calf_twist_01_r {parent=calf_r, loc=<35.813 0.0967 -0.0397>, rot=<0.0237 0.0095 -0.0035 -0.9997>}
 thigh_twist_01_r {parent=thigh_r, loc=<21.6303 0 0>, rot=<0.0013 0 0 -1>}
 dyn_R_custom_calf {parent=thigh_r, loc=<40.3679 -4.4861 -0.0384>, rot=<0.0395 -0.0113 -0.1563 -0.9868>}
 dyn_R_custom_foot {parent=dyn_R_custom_calf, loc=<35.4722 0 0>, rot=<-0.0624 -0.0349 0.4866 -0.8707>}
 dyn_R_custom_ball {parent=dyn_R_custom_foot, loc=<19.3509 0 0>, rot=<0 0 0.4209 -0.9071>}
 dyn_R_custom_ball_corrective {parent=dyn_R_custom_foot, loc=<19.3509 0 0>, rot=<0 0 0.4209 -0.9071>}
 thigh_l_IK {parent=pelvis, loc=<-2.5693 -0.0109 -7.7501>, rot=<0.0292 0.0544 0.0724 -0.9955>}
 dyn_L_custom_calf_IK {parent=thigh_l_IK, loc=<-40.6164 0 0>, rot=<0.0096 -0.0021 -0.2108 -0.9775>}
 dyn_L_custom_foot_IK {parent=dyn_L_custom_calf_IK, loc=<-35.4723 0 0>, rot=<-0.0624 -0.0349 0.4866 -0.8707>}
 thigh_r_IK {parent=pelvis, loc=<-2.5694 -0.0109 7.7501>, rot=<0.0544 -0.0292 0.9955 0.0724>}
 dyn_R_custom_calf_IK {parent=thigh_r_IK, loc=<40.6163 0 0>, rot=<0.0096 -0.0021 -0.2108 -0.9775>}
 dyn_R_custom_foot_IK {parent=dyn_R_custom_calf_IK, loc=<35.4722 0 0>, rot=<-0.0624 -0.0349 0.4866 -0.8707>}
 L_leg_IK_handle {parent=pelvis, loc=<-76.4909 -4.153 -14.9428>, rot=<-0.0127 -0.707 0.0127 -0.707>}
 R_leg_IK_handle {parent=pelvis, loc=<-76.4909 -4.153 14.9428>, rot=<-0.0127 -0.707 0.0127 -0.707>}
 attach {parent=root, loc=<0 0 0>, rot=<0 0 0 1>}
 ik_hand_root {parent=root, loc=<0 0 0>, rot=<0 0 0 1>}
 ik_hand_gun {parent=ik_hand_root, loc=<-51.9727 4.0883 102.6935>, rot=<0.6084 0.1793 0.4613 -0.6205>}
 ik_hand_r {parent=ik_hand_gun, loc=<0 0 0>, rot=<0 0 0 1>}
 ik_hand_l {parent=ik_hand_gun, loc=<53.0285 82.1781 35.2051>, rot=<-0.5102 -0.7906 -0.3387 0>}
 ik_foot_root {parent=root, loc=<0 0 0>, rot=<0 0 0 1>}
 ik_foot_l {parent=ik_foot_root, loc=<15.2608 -4.4176 9.4432>, rot=<0.0357 0.7045 0.0284 -0.7082>}
 ik_foot_r {parent=ik_foot_root, loc=<-15.2608 -4.4176 9.4432>, rot=<0.7082 0.0284 -0.7045 0.0357>}
 root_fp {parent=root, loc=<0 0 0>, rot=<0 0 0 1>}
 attach_fp {parent=root_fp, loc=<0 0 0>, rot=<0 0 0 1>}
 camera_root {parent=root_fp, loc=<0 0 0>, rot=<0 0 0 1>}
 camera {parent=camera_root, loc=<0 0 0>, rot=<0 0 0 1>}
 attach_emote_01 {parent=root, loc=<0 0 0>, rot=<0 0 0 1>}
 attach_emote_02 {parent=root, loc=<0 0 0>, rot=<0 0 0 1>}
 attach_emote_03 {parent=root, loc=<0 0 0>, rot=<0 0 0 1>}
 warp_point_01 {parent=root, loc=<0 0 0>, rot=<0 0 0 1>}
 warp_point_02 {parent=root, loc=<0 0 0>, rot=<0 0 0 1>}
 warp_point_03 {parent=root, loc=<0 0 0>, rot=<0 0 0 1>}
]
variables [
 CustomOuterThumbs {
  type="TArray<FRigElementKeyCollection>"
  default="((Keys=((Type=Bone,Name=\"dyn_L_custom_extraThumb_1\"),(Type=Bone,Name=\"dyn_L_custom_extraThumb_2\"),(Type=Bone,Name=\"dyn_L_custom_extraThumb_3\"))),(Keys=((Type=Bone,Name=\"dyn_R_custom_extraThumb_1\"),(Type=Bone,Name=\"dyn_R_custom_extraThumb_2\"),(Type=Bone,Nam ...(truncated)"
 }
 CustomOuterThumbs Index {type="TArray<FRigElementKey>"}
 OriginalThumbs {
  type="TArray<FRigElementKeyCollection>"
  default=((Keys=((Type=Bone,Name="thumb_01_l"),(Type=Bone,Name="thumb_02_l"),(Type=Bone,Name="thumb_03_l"))),(Keys=((Type=Bone,Name="thumb_01_r"),(Type=Bone,Name="thumb_02_r"),(Type=Bone,Name="thumb_03_r"))))
 }
 OriginalThumbs Index {type="TArray<FRigElementKey>"}
 shoesEquipped {type=bool, default=False}
]
