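# Post-copy AnimBlueprint for the MuffinLadle Gas character body, which has custom front and back arm bones.
# It copies the parent mesh pose into a Control Rig, passing whether an ExtraPart shoe mesh is equipped.
# Two per-bone blends layer DefaultSlot animation onto custom hand bones and EmoteSlot animation onto custom arms and feet.
# Animation curves drive those weights for front-end idles and built-in emotes.
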AnimBlueprint /BRCosmetics/Characters/Player/Male/Medium/Bodies/M_MED_MuffinLadle_Gas/Meshes/M_MED_MuffinLadle_Gas_AnimBP
target_skeleton=&/M_MED_MuffinLadle_Gas_Skeleton
graphs [
 AnimGraph {nodes [
   n1 = Layered blend per bone(
    BasePose=Use cached pose 'startPose'()
    BlendPoses_0=Slot 'DefaultSlot'()
    BlendWeights_0=Get FE Custom Lobby Idle Is Playing Float() -> (FE Custom Lobby Idle Is Playing Float:float)
   ) {layerSetup [{branchFilters [
       {boneName=dyn_L_custom_extraThumb_1}
       {boneName=dyn_R_custom_extraThumb_1}
       {boneName=dyn_L_custom_index_1}
       {boneName=dyn_L_custom_thumb_1}
       {boneName=dyn_L_custom_ring_1}
       {boneName=dyn_L_custom_extraThumb_1}
       {boneName=dyn_R_custom_index_1}
       {boneName=dyn_R_custom_thumb_1}
       {boneName=dyn_R_custom_ring_1}
       {boneName=dyn_R_custom_extraThumb_1}
       {boneName=dyn_L_customFrontArm_wrist}
       {boneName=dyn_L_customBackArm_wrist}
       {boneName=dyn_R_customFrontArm_wrist}
       {boneName=dyn_R_customBackArm_wrist}
      ]}]}
   n2 = Layered blend per bone(
    BasePose=@n1
    BlendPoses_0=Slot 'EmoteSlot'() {slotName=EmoteSlot}
    BlendWeights_0=Get Is Playing Custom Emote() -> (Is Playing Custom Emote:double)
   ) {layerSetup [{branchFilters [
       {boneName=dyn_L_customBackArm_clavicle}
       {boneName=dyn_L_customFrontArm_clavicle}
       {boneName=dyn_R_customBackArm_clavicle}
       {boneName=dyn_R_customFrontArm_clavicle}
       {boneName=dyn_L_custom_ball}
       {boneName=dyn_R_custom_ball}
      ]}]}
   Output Pose(Result=@n2) {layerGroup=None}
   n3 = Control Rig!(
    Source=Copy Pose From Mesh() {bCopyCurves=true, bUseAttachedParent=true}
    Alpha
    shoesEquipped:bool=Get shoesEquipped() -> (shoesEquipped:bool)
   ) -> (Pose) {
    alphaBoolBlend {blendOption=Linear}
    controlRigAssetReference {blueprintRigClass=&/M_MED_MuffinLadle_Gas_CtrlRig.M_MED_MuffinLadle_Gas_CtrlRig_C}
   }
   Save cached pose 'startPose'(Pose=@n3) {cacheName=startPose}
  ]}
 EventGraph {nodes [
   Event BlueprintUpdateAnimation()
   > n1 = Sequence!(execute) -> (then_0, then_1, then_2:Exec)
   > n2 = Branch(Condition=Get bIsPlayingEmote())
   > Set Is Playing Custom Emote(Is Playing Custom Emote:double=GetCurveValue(CurveName=emote_BuiltIn)) -> (Output_Get:double)
   Set Is Playing Custom Emote(execute=@n2.else, Is Playing Custom Emote=0)
   n3 = Get!(execute:Exec=@n1.then_1) -> (then:Exec, else:Exec, ShoeMesh:Skeletal Mesh Component Budgeted Object Reference)
   > n4 = Is Valid(InputObject=Get SkeletalMesh!(self:Skinned Mesh Component Object Reference=@n3) -> (SkeletalMesh:Skeletal Mesh Object Reference))
   Set shoesEquipped(execute=[@n4.Is Not Valid, @n3.else], shoesEquipped:bool=false) -> (Output_Get:bool)
   Set shoesEquipped(execute=@n4.Is Valid, shoesEquipped=true)
   Branch(execute=@n1.then_2, Condition=Get bIsInFrontEnd())
   > Set FE Custom Lobby Idle Is Playing Float(FE Custom Lobby Idle Is Playing Float:float=GetCurveValue(CurveName=Frontend_Idle_Bypass)) -> (Output_Get:float)
   n5 = Event OnCharacterCustomizationLoaded()
   > Set ShoeMesh(ShoeMesh:Skeletal Mesh Component Budgeted Object Reference=Get Array Item with EFort Custom Part Type(Array=@n5.CustomMeshParts, PartType=ExtraPart)) -> (Output_Get:Skeletal Mesh Component Budgeted Object Reference)
  ]}
]
variables [
 FE Custom Lobby Idle Is Playing Float {type=number}
 FE Custom Lobby Layered Pose Blend Float {type=number}
 Is Playing Custom Emote {type=number}
 ShoeMesh {type=SkeletalMeshComponentBudgeted}
 shoesEquipped {type=boolean}
]
